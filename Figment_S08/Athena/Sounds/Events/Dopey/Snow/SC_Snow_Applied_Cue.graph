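# Looping Sound Cue for the "snow applied" sound in the Dopey event, routed to the Fort_Events_Mapwide sound class.
# A looping Snow_Complete_Loop wave player feeds an Enveloper whose volume curve starts boosted and drops to unity over three seconds.
# Cue-level attenuation overrides set distance falloff, stereo spread, distance-based reverb and interpolated listener focus, which boosts volume in focus and reduces it off-focus.
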SoundCue /Figment_S08/Athena/Sounds/Events/Dopey/Snow/SC_Snow_Applied_Cue
root=Enveloper_0
nodes [
 {
  id=Enveloper_0
  children [WavePlayer_1]
  properties {pitchCurve {editorCurveData {keys [{time=0}]}}, volumeCurve {editorCurveData {keys [<at 0 2>,<at 3 1>]}}}
 }
 {id=WavePlayer_1, properties {bLooping=true, soundWaveAssetPtr=&/Snow_Complete_Loop}}
]
properties {
 bOverrideAttenuation=true
 maxDistance=2524
 soundClassObject=/Figment_S08/Athena/Sounds/Events/SoundClasses/Fort_Events_Mapwide
 attenuationOverrides {
  bEnableFocusInterpolation=true
  bEnableListenerFocus=true
  falloffDistance=1500
  focusAttackInterpSpeed=5
  focusAzimuth=10
  focusReleaseInterpSpeed=5
  focusVolumeAttenuation=2
  nonFocusAzimuth=30
  nonFocusVolumeAttenuation=0.3
  reverbDistanceMax=3000
  reverbDistanceMin=1000
  reverbWetLevelMax=10
  reverbWetLevelMin=0
  stereoSpread=512
  attenuationShapeExtents {x=1024}
 }
}
