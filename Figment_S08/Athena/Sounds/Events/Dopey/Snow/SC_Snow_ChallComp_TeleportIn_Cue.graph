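# Sound cue for a challenge-complete teleport-in sound in the Snow event set, played through a mapwide events sound class.
# A mixer combines close and distant attenuation branches, with the close branch playing a single wave after a fixed delay.
# The distant branch has no child connected, so only the close layer produces sound.
# The cue carries a boosted volume and a very large max distance.
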SoundCue /Figment_S08/Athena/Sounds/Events/Dopey/Snow/SC_Snow_ChallComp_TeleportIn_Cue
root=Mixer_0
nodes [
 {id=Attenuation_0, children [Delay_0], properties {attenuationSettings=&/ATT_Snow_MLTI_Close_Att}}
 {id=Attenuation_1, children [null], properties {attenuationSettings=&/ATT_Snow_MLTI_Distant_Att}}
 {id=Delay_0, children [WavePlayer_0], properties {delayMax=1, delayMin=1}}
 {id=Mixer_0, children [Attenuation_0, Attenuation_1], properties {inputVolume [1,1]}}
 {id=WavePlayer_0, properties {soundWaveAssetPtr=&/Snow_ChallComp_TeleportIn_Close_01}}
]
properties {maxDistance=2097152, soundClassObject=/Figment_S08/Athena/Sounds/Events/SoundClasses/Fort_Events_Mapwide, volumeMultiplier=1.5}
