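# SC_Snow_Lock_Start_Cue is a sound cue in the Dopey Snow event audio, named for a lock start moment.
# A mixer blends two equal-volume branches, one using close attenuation and the other distant attenuation.
# Both attenuation nodes have empty inputs, so the cue references no sound waves and plays nothing as it stands.
# It routes to the mapwide events sound class with a very large max distance.
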SoundCue /Figment_S08/Athena/Sounds/Events/Dopey/Snow/SC_Snow_Lock_Start_Cue
root=Mixer_0
nodes [
 {id=Attenuation_0, children [null], properties {attenuationSettings=&/ATT_Snow_MLTI_Close_Att}}
 {id=Attenuation_1, children [null], properties {attenuationSettings=&/ATT_Snow_MLTI_Distant_Att}}
 {id=Mixer_0, children [Attenuation_0, Attenuation_1], properties {inputVolume [1,1]}}
]
properties {maxDistance=2097152, soundClassObject=/Figment_S08/Athena/Sounds/Events/SoundClasses/Fort_Events_Mapwide}
