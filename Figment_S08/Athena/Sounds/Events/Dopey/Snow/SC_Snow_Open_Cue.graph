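# Sound cue for the snow "open" sound in the Figment S08 Dopey event.
# It plays a single wave, Snow_Complete_01, through an attenuation node using the snow close attenuation settings, feeding a root mixer whose second input is empty.
# The cue uses the map-wide events sound class, with a boosted volume and a large max distance.
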SoundCue /Figment_S08/Athena/Sounds/Events/Dopey/Snow/SC_Snow_Open_Cue
root=Mixer_0
nodes [
 {
  id=Attenuation_0
  children [WavePlayer_6]
  properties {attenuationSettings=&/ATT_Snow_MLTI_Close_Att}
 }
 {id=Mixer_0, children [Attenuation_0, null], properties {inputVolume [1,1]}}
 {id=WavePlayer_6, properties {soundWaveAssetPtr=&/Snow_Complete_01}}
]
properties {maxDistance=22500, soundClassObject=/Figment_S08/Athena/Sounds/Events/SoundClasses/Fort_Events_Mapwide, volumeMultiplier=1.5}
