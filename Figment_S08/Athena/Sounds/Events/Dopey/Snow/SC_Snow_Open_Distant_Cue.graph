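# SoundCue for the distant layer of a "Snow Open" event sound in Figment Season 8, routed through the map-wide events sound class.
# A single distant snow wave passes through its own attenuation node and a fixed two-second delay into a mixer whose other input is empty.
# A concurrency override allows one instance and blocks new ones, with boosted volume and a very large max distance.
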SoundCue /Figment_S08/Athena/Sounds/Events/Dopey/Snow/SC_Snow_Open_Distant_Cue
root=Mixer_0
nodes [
 {
  id=Attenuation_1
  children [WavePlayer_1]
  properties {attenuationSettings=&/ATT_Snow_MLTI_Distant_Att}
 }
 {id=Delay_0, children [Attenuation_1], properties {delayMax=2, delayMin=2}}
 {id=Mixer_0, children [null,Delay_0], properties {inputVolume [1,1]}}
 {id=WavePlayer_1, properties {soundWaveAssetPtr=&/Snow_Complete_Distant_01}}
]
properties {
 bOverrideConcurrency=true
 maxDistance=2097152
 soundClassObject=/Figment_S08/Athena/Sounds/Events/SoundClasses/Fort_Events_Mapwide
 volumeMultiplier=1.5
 concurrencyOverrides {maxCount=1, resolutionRule=PreventNew}
}
