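# Map-wide ambience cue for the post stage of the Dopey Snow event, layering a snow loop with volcano sounds.
# The close-attenuated layer mixes a looping exterior bed with randomly delayed earth-movement or thunder one-shots.
# Distant layers, each separately attenuated, add the snow loop and sparser random distant rumbles.
# Looping, Delay and Random nodes generate the irregular one-shots, and concurrency allows one instance, stopping the oldest.
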SoundCue /Figment_S08/Athena/Sounds/Events/Dopey/Snow/SC_Snow_Post_Cue
root=Mixer_0
nodes [
 {id=Attenuation_0, children [Mixer_2], properties {attenuationSettings=&/ATT_Snow_MLTI_Close_Att}}
 {
  id=Attenuation_2
  children [WavePlayer_1]
  properties {attenuationSettings=&/Snow_MLTI_Distant_Focus_Att}
 }
 {
  id=Attenuation_3
  children [Looping_0]
  properties {attenuationSettings=&/ATT_Snow_MLTI_Distant_Att}
 }
 {id=Delay_0, children [Random_0], properties {delayMax=30, delayMin=10}}
 {id=Delay_1, children [Random_3], properties {delayMax=10, delayMin=3}}
 {id=Looping_0, children [Delay_0]}
 {id=Looping_1, children [Delay_1]}
 {id=Mixer_0, children [Attenuation_0, Mixer_1], properties {inputVolume [2,2]}}
 {id=Mixer_1, children [Attenuation_2, Attenuation_3], properties {inputVolume [1,1.5]}}
 {id=Mixer_2, children [WavePlayer_6,Looping_1], properties {inputVolume [1,1]}}
 {id=Random_0, children [WavePlayer_0,WavePlayer_2,WavePlayer_3,WavePlayer_4,WavePlayer_5], properties {weights [1,1,1,1,1]}}
 {id=Random_2, children [WavePlayer_11, WavePlayer_12], properties {weights [1,1]}}
 {id=Random_3, children [Random_5,Random_2], properties {weights [1,1]}}
 {id=Random_5, children [WavePlayer_15, WavePlayer_16, WavePlayer_17, WavePlayer_18], properties {weights [1,1,1,1]}}
 {id=WavePlayer_0, properties {soundWaveAssetPtr=/Game/Athena/Sounds/Ambient/Volcano/Amb_Volcano_Earth_Move_Rumble_Dist_01}}
 {id=WavePlayer_1, properties {bLooping=true, soundWaveAssetPtr=&/Snow_Post_LOOP_01}}
 {id=WavePlayer_11, properties {soundWaveAssetPtr=/Figment_S08/Athena/Sounds/Ambient/Volcano/Amb_Volcano_Thunder_Rumble_Med_Dist_02}}
 {id=WavePlayer_12, properties {soundWaveAssetPtr=/Game/Athena/Sounds/Ambient/Volcano/Amb_Volcano_Thunder_Rumble_Med_Dist_01}}
 {id=WavePlayer_15, properties {soundWaveAssetPtr=/FNGameplayCues/Athena/Sounds/Ambient/Volcano/Amb_Volcano_Earth_Move_01}}
 {id=WavePlayer_16, properties {soundWaveAssetPtr=/Figment_S08/Athena/Sounds/Ambient/Volcano/Amb_Volcano_Earth_Move_02}}
 {id=WavePlayer_17, properties {soundWaveAssetPtr=/Figment_S08/Athena/Sounds/Ambient/Volcano/Amb_Volcano_Earth_Move_03}}
 {id=WavePlayer_18, properties {soundWaveAssetPtr=/Figment_S08/Athena/Sounds/Ambient/Volcano/Amb_Volcano_Earth_Move_04}}
 {id=WavePlayer_2, properties {soundWaveAssetPtr=/Game/Athena/Sounds/Ambient/Volcano/Amb_Volcano_Earth_Move_Rumble_Dist_02}}
 {id=WavePlayer_3, properties {soundWaveAssetPtr=/Game/Athena/Sounds/Ambient/Volcano/Amb_Volcano_Earth_Move_Rumble_Dist_03}}
 {id=WavePlayer_4, properties {soundWaveAssetPtr=/Game/Athena/Sounds/Ambient/Volcano/Amb_Volcano_Earth_Move_Rumble_Dist_04}}
 {id=WavePlayer_5, properties {soundWaveAssetPtr=/Game/Athena/Sounds/Ambient/Volcano/Amb_Volcano_Earth_Move_Rumble_Dist_05}}
 {id=WavePlayer_6, properties {bLooping=true, soundWaveAssetPtr=/Figment_S08/Athena/Sounds/Ambient/Volcano/Amb_Volcano_ExteriorLoop_01}}
]
properties {
 bOverrideConcurrency=true
 maxDistance=2097152
 soundClassObject=/Figment_S08/Athena/Sounds/Events/SoundClasses/Fort_Events_Mapwide
 volumeMultiplier=1.5
 concurrencyOverrides {maxCount=1, resolutionRule=StopOldest}
}
