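# Sound cue that plays the snow sequence audio, apparently for the Dopey snow event in Figment Season 08 Athena.
# A single wave player for Snow_Sequence_01 feeds an attenuation node using the Snow_MLTI_CloseDistant_Att settings.
# The cue routes to the map-wide events sound class, with a boosted volume and a very large max distance.
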SoundCue /Figment_S08/Athena/Sounds/Events/Dopey/Snow/SC_Snow_Sequence_Cue
root=Attenuation_0
nodes [
 {
  id=Attenuation_0
  children [WavePlayer_0]
  properties {attenuationSettings=&/Snow_MLTI_CloseDistant_Att}
 }
 {id=WavePlayer_0, properties {soundWaveAssetPtr=&/Snow_Sequence_01}}
]
properties {maxDistance=2097152, soundClassObject=/Figment_S08/Athena/Sounds/Events/SoundClasses/Fort_Events_Mapwide, volumeMultiplier=1.5}
