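# Sound cue for a looping snow ambience in the Dopey snow event under the Figment S08 Athena content, judging by its paths and names.
# A looping wave player plays the Snow_Dest_Amb_01 loop through an attenuation node using shared close-to-distant snow attenuation settings.
# The cue boosts volume, uses the map-wide events sound class, and has a very large max distance.
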SoundCue /Figment_S08/Athena/Sounds/Events/Dopey/Snow/SC_Snow_Sequence_Loop_Cue
root=Attenuation_0
nodes [
 {
  id=Attenuation_0
  children [WavePlayer_1]
  properties {attenuationSettings=&/Snow_MLTI_CloseDistant_Att}
 }
 {id=WavePlayer_1, properties {bLooping=true, soundWaveAssetPtr=&/Snow_Dest_Amb_01_LOOP}}
]
properties {maxDistance=2097152, soundClassObject=/Figment_S08/Athena/Sounds/Events/SoundClasses/Fort_Events_Mapwide, volumeMultiplier=1.5}
