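# Sound cue for a looping orb ambience in the Snow destination set of the Figment S08 Dopey event sounds.
# A single looping wave player feeds the root directly, with boosted volume and the Fort_Nexus_Transition sound class.
# Attenuation is overridden on the cue with a large inner radius, falloff, stereo spread and distance-driven reverb send.
# Interpolated listener focus makes the orb louder when the listener faces it.
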SoundCue /Figment_S08/Athena/Sounds/Events/Dopey/Snow/Snow_Dest_Amb_1_Cue
root=WavePlayer_0
nodes [{id=WavePlayer_0, properties {bLooping=true, soundWaveAssetPtr=&/Snow_Dest_Amb_05_Orb_LOOP}}]
properties {
 bOverrideAttenuation=true
 maxDistance=11000
 soundClassObject=/Figment_S08/Sounds/FortSoundClassesAndModes/Fort_Nexus_Transition
 volumeMultiplier=2
 attenuationOverrides {
  bEnableFocusInterpolation=true
  bEnableListenerFocus=true
  falloffDistance=3000
  focusAttackInterpSpeed=5
  focusAzimuth=20
  focusReleaseInterpSpeed=5
  focusVolumeAttenuation=2
  nonFocusAzimuth=45
  reverbDistanceMax=3000
  reverbDistanceMin=1000
  reverbWetLevelMax=10
  reverbWetLevelMin=0
  stereoSpread=1024
  attenuationShapeExtents {x=8000}
 }
}
