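# Sound cue for a looping snow destination ambience, part of the Dopey Snow event audio in Figment S08.
# A single looping wave player feeds the Snow_Dest_Desk_01_LOOP sound straight to the output at half volume, routed to the Fort_Nexus_Transition sound class.
# Overridden attenuation adds distance falloff, low-pass filtering, a wide stereo spread, and interpolated listener focus that changes non-focus volume.
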SoundCue /Figment_S08/Athena/Sounds/Events/Dopey/Snow/Snow_Dest_Amb_3_Cue
root=WavePlayer_0
nodes [{id=WavePlayer_0, properties {bLooping=true, soundWaveAssetPtr=&/Snow_Dest_Desk_01_LOOP}}]
properties {
 bOverrideAttenuation=true
 maxDistance=1506.2095
 soundClassObject=/Figment_S08/Sounds/FortSoundClassesAndModes/Fort_Nexus_Transition
 volumeMultiplier=0.5
 attenuationOverrides {
  bAttenuateWithLPF=true
  bEnableFocusInterpolation=true
  bEnableListenerFocus=true
  falloffDistance=810.2751
  focusAttackInterpSpeed=5
  focusAzimuth=40
  focusReleaseInterpSpeed=5
  lPFFrequencyAtMax=6000
  lPFFrequencyAtMin=7000
  lPFRadiusMax=0
  lPFRadiusMin=0
  nonFocusAzimuth=70
  nonFocusVolumeAttenuation=0.3
  reverbDistanceMax=0
  reverbDistanceMin=0
  reverbWetLevelMax=10
  reverbWetLevelMin=0
  stereoSpread=1024
  attenuationShapeExtents {x=695.9344}
 }
}
