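# Snow destination ambience SoundCue for a descend sound in the Dopey event, routed to the Fort_Nexus_Transition sound class.
# It has no sound nodes and a null root, so the file itself shows no wave to play.
# It carries a boosted volume and inline attenuation overrides for distance falloff, wide stereo spread and distance-based reverb send.
# Interpolated focus settings boost the sound when faced and reduce it off-axis.
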SoundCue /Figment_S08/Athena/Sounds/Events/Dopey/Snow/Snow_Dest_Amb_Descend_2_Cue
root=null
nodes []
properties {
 bOverrideAttenuation=true
 maxDistance=2806.5332
 soundClassObject=/Figment_S08/Sounds/FortSoundClassesAndModes/Fort_Nexus_Transition
 volumeMultiplier=1.5
 attenuationOverrides {
  bEnableFocusInterpolation=true
  falloffDistance=2110.5989
  focusAttackInterpSpeed=5
  focusAzimuth=20
  focusReleaseInterpSpeed=0.5
  focusVolumeAttenuation=2
  nonFocusAzimuth=45
  nonFocusVolumeAttenuation=0.3
  reverbDistanceMax=3000
  reverbDistanceMin=1000
  reverbWetLevelMax=10
  reverbWetLevelMin=0
  stereoSpread=1024
  attenuationShapeExtents {x=695.9344}
 }
}
