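# Sound cue for the Figment S08 Dopey snow event, named as destruction ambience for a winner moment, that plays one global snow-break wave at doubled volume.
# A single wave player feeds an Attenuation node that overrides distance falloff and spatialization off, so the sound plays non-positionally at a constant level.
# It routes through the Fort_Nexus_Transition sound class.
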SoundCue /Figment_S08/Athena/Sounds/Events/Dopey/Snow/Snow_Dest_Amb_Winner_2_Cue
root=Attenuation_1
nodes [
 {
  id=Attenuation_1
  children [WavePlayer_0]
  properties {bOverrideAttenuation=true, attenuationOverrides {bAttenuate=false, bSpatialize=false, falloffDistance=1000000}}
 }
 {id=WavePlayer_0, properties {soundWaveAssetPtr=&/Snow_Break_Global_01}}
]
properties {
 maxDistance=1000400
 soundClassObject=/Figment_S08/Sounds/FortSoundClassesAndModes/Fort_Nexus_Transition
 volumeMultiplier=2
 attenuationOverrides {
  bAttenuate=false
  bEnableFocusInterpolation=true
  bSpatialize=false
  falloffDistance=10000
  focusAttackInterpSpeed=5
  focusAzimuth=20
  focusReleaseInterpSpeed=0.5
  focusVolumeAttenuation=2
  nonFocusAzimuth=45
  nonFocusVolumeAttenuation=0.3
  reverbDistanceMax=3000
  reverbDistanceMin=1000
  reverbWetLevelMax=10
  reverbWetLevelMin=0
  stereoSpread=1024
  attenuationShapeExtents {x=695.9344}
 }
}
