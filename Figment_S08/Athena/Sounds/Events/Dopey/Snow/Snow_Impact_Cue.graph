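# Sound cue for snow pillar impact effects in the Figment Season 8 Dopey snow event audio.
# A modulator feeds an equally weighted random node that picks one of five Snow_PillarImpact wave variants.
# It uses a snow impact attenuation asset, the Fort_Nexus_Transition sound class and a raised volume.
# Overridden concurrency caps simultaneous instances at a few, stopping the oldest first.
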SoundCue /Figment_S08/Athena/Sounds/Events/Dopey/Snow/Snow_Impact_Cue
root=Modulator_0
nodes [
 {id=Modulator_0, children [Random_0]}
 {id=Random_0, children [WavePlayer_0,WavePlayer_1,WavePlayer_2,WavePlayer_3,WavePlayer_4], properties {weights [1,1,1,1,1]}}
 {id=WavePlayer_0, properties {soundWaveAssetPtr=&/Snow_PillarImpact_01}}
 {id=WavePlayer_1, properties {soundWaveAssetPtr=&/Snow_PillarImpact_02}}
 {id=WavePlayer_2, properties {soundWaveAssetPtr=&/Snow_PillarImpact_03}}
 {id=WavePlayer_3, properties {soundWaveAssetPtr=&/Snow_PillarImpact_04}}
 {id=WavePlayer_4, properties {soundWaveAssetPtr=&/Snow_PillarImpact_05}}
]
properties {
 attenuationSettings=&/Snow_Impacts_Att
 bOverrideConcurrency=true
 maxDistance=6000
 soundClassObject=/Figment_S08/Sounds/FortSoundClassesAndModes/Fort_Nexus_Transition
 volumeMultiplier=1.5
 attenuationOverrides {falloffDistance=3000, attenuationShapeExtents {x=0}}
 concurrencyOverrides {maxCount=4, resolutionRule=StopOldest}
}
