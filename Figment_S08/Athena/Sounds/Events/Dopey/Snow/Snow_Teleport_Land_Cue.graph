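# Sound cue for the landing after the Snow teleport in the Figment Season 8 Dopey event audio.
# A single wave player feeds the Snow_Land wave straight to the output, with no randomization, mixing or modulation nodes.
# It routes through the Fort_Nexus_Transition sound class and sets a very large max distance.
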SoundCue /Figment_S08/Athena/Sounds/Events/Dopey/Snow/Snow_Teleport_Land_Cue
root=WavePlayer_0
nodes [{id=WavePlayer_0, properties {soundWaveAssetPtr=&/Snow_Land}}]
properties {maxDistance=2097152, soundClassObject=/Figment_S08/Sounds/FortSoundClassesAndModes/Fort_Nexus_Transition, volumeMultiplier=1}
