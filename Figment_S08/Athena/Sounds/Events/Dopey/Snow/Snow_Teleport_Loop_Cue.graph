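# Layered snow-themed teleport sound cue for the Dopey event, routed to the Nexus transition sound class with overridden attenuation.
# A Mixer plays a one-shot Snow_Enter alongside three looping ambience waves, with the ground layer boosted.
# Envelopers fade one ambience out while moving its pitch, and cycle another through an endless slow swell and fade so the loop keeps evolving.
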SoundCue /Figment_S08/Athena/Sounds/Events/Dopey/Snow/Snow_Teleport_Loop_Cue
root=Mixer_0
nodes [
 {
  id=Enveloper_0
  children [WavePlayer_1]
  properties {pitchCurve {editorCurveData {keys [{time=0}, <at 10 0.8>]}}, volumeCurve {editorCurveData {keys [<at 0 1>,<at 20 0>]}}}
 }
 {
  id=Enveloper_1
  children [WavePlayer_3]
  properties {
   bLoop=true
   bLoopIndefinitely=true
   loopEnd=40
   pitchCurve {editorCurveData {keys [{time=0}]}}
   volumeCurve {editorCurveData {keys [<at 0 0>,<at 20 1>,<at 40 0>]}}
  }
 }
 {id=Mixer_0, children [WavePlayer_2,Enveloper_0,Enveloper_1,WavePlayer_0], properties {inputVolume [2,1,1,1]}}
 {id=WavePlayer_0, properties {soundWaveAssetPtr=/Figment_S08/DirtShell_8_5_1/Athena/Sounds/Events/Dopey/Snow/Snow_Enter}}
 {id=WavePlayer_1, properties {bLooping=true, soundWaveAssetPtr=&/Snow_Dest_Amb_01_LOOP}}
 {id=WavePlayer_2, properties {bLooping=true, soundWaveAssetPtr=&/Snow_Dest_Amb_05_Ground_LOOP}}
 {id=WavePlayer_3, properties {bLooping=true, soundWaveAssetPtr=&/Snow_Dest_Amb_02_LOOP}}
]
properties {
 bOverrideAttenuation=true
 maxDistance=4000
 soundClassObject=/Figment_S08/Sounds/FortSoundClassesAndModes/Fort_Nexus_Transition
 volumeMultiplier=1
}
