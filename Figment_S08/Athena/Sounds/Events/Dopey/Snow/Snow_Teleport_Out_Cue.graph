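# Sound cue for the Snow teleport-out moment in the Figment S08 Dopey event, judging by its name and path.
# It plays a single wave player that reuses the existing third-person RiftPortal teleport exit sound.
# The cue overrides attenuation and routes through the Fort_Nexus_Transition sound class.
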SoundCue /Figment_S08/Athena/Sounds/Events/Dopey/Snow/Snow_Teleport_Out_Cue
root=WavePlayer_3
nodes [{id=WavePlayer_3, properties {soundWaveAssetPtr=/Game/Sounds/Fort_GamePlay_Sounds/Player/RiftPortal/RiftPortal_Teleport_Exit_3P_03}}]
properties {bOverrideAttenuation=true, maxDistance=4000, soundClassObject=/Figment_S08/Sounds/FortSoundClassesAndModes/Fort_Nexus_Transition}
